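; Commerce prop Blueprint for a large market stall half that tints its meshes from its building texture data.
; The construction script runs the parent's, then, if texture data exists, stores the static mesh component and its direct children in ExtraComponents.
; Each is cast to a static mesh component and given the tint of texture entry index 1, though the guard only checks for a non-empty array.
; BeginPlay, overlap and Tick events are empty.
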(event EventBeginPlay)

(event Collision|EventActorBeginOverlap
  (params (OtherActor :type Actor%20Object%20Reference)))

(event EventTick
  (params (DeltaSeconds :type Float%20%28single-precision%29)))

(fn ConstructionScript
  (bind _staticmeshcomponent (Variables|Debug|GetStaticMeshComponent))
  (bind _extracomponents (Variables|Default|GetExtraComponents))
  (|CallParentFunction)
  (bind _outtexturedataarray (Building|GetBuildingTextureData))
  (if (> (Utilities|Array|Length _outtexturedataarray) 0)
    (Variables|Default|SetExtraComponents (Components|GetChildrenComponents _staticmeshcomponent false))
    (Utilities|Array|Add _extracomponents _staticmeshcomponent)
    (for _array_element _extracomponents
      (bind _asstatic_mesh_component (Utilities|Casting|CastToStaticMeshComponent _array_element))
      (Default|SetColoronSecondaryMesh (Math|Conversions|ToLinearColor(Color) (Variables|/Script/FortniteGame.BuildingTextureData|GetTintColor (Utilities|Array|Get(acopy) _outtexturedataarray 1))) _asstatic_mesh_component))))
