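; Large commerce market stall building prop that tints its gate meshes and main mesh from its building texture data.
; After the parent construction script, if texture data exists, it colors the gate component and its direct children with the second entry's tint.
; The same loop writes that tint, desaturated by SSS_Desaturation, to the SSS tint parameter on the main mesh's dynamic material.
; BeginPlay, overlap and Tick events are empty.
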(event EventBeginPlay)

(event Collision|EventActorBeginOverlap
  (params (OtherActor :type Actor%20Object%20Reference)))

(event EventTick
  (params (DeltaSeconds :type Float%20%28single-precision%29)))

(fn ConstructionScript
  (bind _sm_commerce_marketstalllargegate_a (Variables|Default|GetSM_Commerce_MarketStallLargeGate_A))
  (bind _extracomponents (LocalVariables|GetExtraComponents))
  (|CallParentFunction)
  (bind _outtexturedataarray (Building|GetBuildingTextureData))
  (if (> (Utilities|Array|Length _outtexturedataarray) 0)
    (LocalVariables|SetExtraComponents (Components|GetChildrenComponents _sm_commerce_marketstalllargegate_a false))
    (Utilities|Array|Add _extracomponents _sm_commerce_marketstalllargegate_a)
    (for _array_element _extracomponents
      (bind _asstatic_mesh_component (Utilities|Casting|CastToStaticMeshComponent _array_element))
      (Default|SetColoronSecondaryMesh (Math|Conversions|ToLinearColor(Color) (Variables|/Script/FortniteGame.BuildingTextureData|GetTintColor (Utilities|Array|Get(acopy) _outtexturedataarray 1))) _asstatic_mesh_component)
      (Rendering|Material|SetVectorParameterValue (Class|BTechArtGlobalLibrary|GetDynamicMaterialInstance (Variables|Debug|GetStaticMeshComponent)) "SFX_SSS Tint(Expand)" (Math|Color|Desaturate(LinearColor) (Math|Conversions|ToLinearColor(Color) (Variables|/Script/FortniteGame.BuildingTextureData|GetTintColor (Utilities|Array|Get(acopy) _outtexturedataarray 1))) (Variables|Default|GetSSS_Desaturation)) :!class /Script/Engine.MaterialInstanceDynamic))))
